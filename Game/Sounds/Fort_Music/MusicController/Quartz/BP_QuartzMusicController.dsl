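; Quartz music controller Blueprint that arbitrates which Quartz music track component is the current track.
; RequestTrackPlayback only answers whether a track may play: true when no track is current or its priority is strictly higher.
; PlayTrack ignores the current track; for any other it queues a stop of the old, queues the new start and records it.
; StopCurrentTrack queues the stop and clears the current reference.
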(fn GetCurrentTrack
  (return (Variables|Default|GetCurrentTrack)))

(fn PlayTrack
  (params (Track%20Component :type Quartz%20Music%20Track%20Component%20Object%20Reference))
  (if (Utilities|NotEqual(Object) Track%20Component (Variables|Default|GetCurrentTrack))
    (Utilities|FlowControl|Sequence
      (:then_0
        (CallFunction|StopCurrentTrack))
      (:then_1
        (Control|QueueTrackStart Track%20Component)
        (Variables|Default|SetCurrentTrack Track%20Component)))))

(fn RequestTrackPlayback
  (params (Requesting%20Track :type Quartz%20Music%20Track%20Component%20Object%20Reference))
  (bind _currenttrack (Variables|Default|GetCurrentTrack)
    (:then
      (return (> (Class|QuartzMusicTrackComponent|GetTrackPriority Requesting%20Track :!class /Game/Sounds/Fort_Music/MusicController/Quartz/QuartzMusicTrackComponent.QuartzMusicTrackComponent_C) (Class|QuartzMusicTrackComponent|GetTrackPriority _currenttrack :!class /Game/Sounds/Fort_Music/MusicController/Quartz/QuartzMusicTrackComponent.QuartzMusicTrackComponent_C))))
    (:else
      (return true))))

(fn StopCurrentTrack
  (bind _currenttrack (Variables|Default|GetCurrentTrack)
    (:then
      (Control|QueueTrackStop _currenttrack)
      (Variables|Default|SetCurrentTrack 0))
    (:else)))

(fn ConstructionScript)
