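; Music track component for the Quartz music controller that starts, stops and parameterizes one MetaSound or SoundCue track.
; QueueTrackStart creates a clock named after the sound, notifies the controller, plays quantized, and registers audio analysis and music events.
; QueueTrackStop flags a stop that fades out and deletes the clock on the next beat, unless a new start cancels it.
; Debug mode plays a test tone each beat.
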(event EventBeginPlay
  (bind _asbp_quartz_music_controller (Utilities|Casting|CastToBP_QuartzMusicController (Components|GetOwner)))
  (Variables|Default|SetControllerActor _asbp_quartz_music_controller))

(event Custom|SetDebugEnabled
  (params (Debug :type Boolean))
  (Variables|Default|SetbDebugEnabled Debug))

(event Custom|SetStemVolume
  (params (Index :type Integer) (Value :type Float%20%28double-precision%29))
  (bind _audiocomponent (Variables|Components|GetAudioComponent)
    (:then
      (Audio|Parameter|SetFloatParameter _audiocomponent (Utilities|Array|Get(aref) (Variables|TrackSettings|GetTrackParameters) Index) Value :!class /Script/AudioExtensions.AudioParameterControllerInterface))
    (:else)))

(event EventEndPlay
  (params (EndPlayReason :type EEndPlayReason%20Enum :default "Destroyed"))
  (bind _audiocomponent (Variables|Components|GetAudioComponent)
    (:then
      (Audio|Components|Audio|FadeOut _audiocomponent (Variables|TrackSettings|GetTrackFadeOutTime) 0.0 "Linear" :!class /Script/Engine.AudioComponent))
    (:else)))

(fn GetAudioComponent
  (return (Variables|Components|GetAudioComponent)))

(fn GetClockHandle
  (return (Variables|ClockSettings|GetClockInst)))

(fn GetIsPlaying
  (bind _audiocomponent (Variables|Components|GetAudioComponent)
    (:then
      (return (Audio|Components|Audio|IsPlaying _audiocomponent)))
    (:else
      (return false))))

(fn GetPlayState
  (bind _audiocomponent (Variables|Components|GetAudioComponent)
    (:then
      (return (Audio|Components|Audio|GetPlayState _audiocomponent)))
    (:else
      (return "Stopped"))))

(fn GetTrackPriority
  (return (Variables|TrackSettings|GetTrackPriority)))

(fn GetTrackSound
  (bind _trackmetasound (Variables|TrackSettings|GetTrackMetaSound)
    (:then
      (return _trackmetasound))
    (:else
      (return (Variables|TrackSettings|GetTrackSoundCue)))))

(fn IsEnabled
  (return true))

(fn OnQuartzClockDelegate
  (params (EventType :type EQuartzCommandDelegateSubType%20Enum :default "CommandOnFailedToQueue") (Name :type Name))
  (switch Utilities|FlowControl|Switch|SwitchonEQuartzCommandDelegateSubType EventType
    (:CommandOnQueued
      (QuartzClock|ResumeClock (Variables|ClockSettings|GetClockInst)))))

(fn OnQuartzQuantizationEvent
  (params (ClockName :type Name) (QuantizationType :type EQuartzCommandQuantization%20Enum :default "Bar") (NumBars :type Integer) (Beat :type Integer) (BeatFraction :type Float%20%28single-precision%29))
  (if (Utilities|Name|Equal(Name) ClockName (Variables|ClockSettings|GetClockName))
    (switch Utilities|FlowControl|Switch|SwitchonEQuartzCommandQuantization QuantizationType
      (:Beat
        (Utilities|FlowControl|Sequence
          (:then_0
            (if (Variables|Default|GetbQueueStop)
              (bind _audiocomponent (Variables|Components|GetAudioComponent)
                (:then
                  (Audio|Components|Audio|FadeOut _audiocomponent (Variables|TrackSettings|GetTrackFadeOutTime) 0.0 "Linear" :!class /Script/Engine.AudioComponent)
                  (Variables|Default|SetbQueueStop false)
                  (Variables|Components|SetAudioComponent 0)
                  (Development|PrintString "Track Stopped")
                  (QuartzClockHandle|DeleteClockbyHandle (WorldSubsystems|GetQuartzSubsystem) (Variables|ClockSettings|GetClockInst))
                  (Variables|ClockSettings|SetClockInst 0))
                (:else))))
          (:then_1
            (if (Variables|Default|GetbDebugEnabled)
              (Audio|PlaySound2D "/Game/Sounds/Fort_Test_Sounds/Tones/1k_test_tone.1k_test_tone"))))))))

(fn QueueTrackStart
  (bind _sound (CallFunction|GetTrackSound))
  (bind _returnvalue (Utilities|GetDisplayName _sound))
  (bind _returnvalue_2 (Fortnite|Athena|GameState|GetGameStateAthena))
  (Utilities|IsValid _returnvalue_2
    (:Is%20Valid
      (if (and (not (GameplayTags|HasAnyTags (Class|FortGameStateAthena|GetAthenaPlaylistContextTags _returnvalue_2) (Variables|TrackSettings|GetTrackDisableTags))) (CallFunction|IsEnabled))
        _))
    (:Is%20Not%20Valid
      _))
  (if (Utilities|IsValid (CallFunction|GetTrackSound) :!class /Script/Engine.KismetSystemLibrary)
    (Utilities|FlowControl|Sequence
      (:then_0
        (Variables|ClockSettings|GetClockInst
          (:then)
          (:else
            (bind _output_get (Variables|ClockSettings|SetClockName (Utilities|String|StringToName _returnvalue)))
            (bind _returnvalue_3 (QuartzClockHandle|CreateNewClock (WorldSubsystems|GetQuartzSubsystem) _output_get (Variables|ClockSettings|GetClockSettings) false true))
            (Development|PrintString _returnvalue)
            (bind _output_get_2 (Variables|ClockSettings|SetClockInst _returnvalue_3))
            (bind _controlleractor (Variables|Default|GetControllerActor)
              (:then
                (Default|CallOnClockCreated _controlleractor (Variables|ClockSettings|GetClockName) _output_get_2))
              (:else
                _))
            (Quantization|SetBeatsPerMinute :self (Variables|ClockSettings|GetClockInst) :BeatsPerMinute (Variables|ClockSettings|GetClockBPM))
            (QuartzClock|SubscribetoAllQuantizationEvents (Variables|ClockSettings|GetClockInst) (EventDispatchers|CreateEvent :!event OnQuartzQuantizationEvent)))))
      (:then_1
        ;[1 1 1] Cancel the queued-to-stop AudioComponent
        (Variables|Components|GetAudioComponent
          (:then
            (Variables|Default|SetbQueueStop false))
          (:else
            (bind _returnvalue_4 (Audio|CreateSound2D (CallFunction|GetTrackSound)))
            (bind _output_get_3 (Variables|Components|SetAudioComponent _returnvalue_4))
            (Audio|Components|Audio|PlayQuantized _output_get_3 (Variables|ClockSettings|GetClockInst) (Variables|ClockSettings|GetClockQuantizationBoundary) (EventDispatchers|CreateEvent :!event OnQuartzClockDelegate) 0.0 (Variables|TrackSettings|GetTrackFadeInTime))
            (Development|PrintString "Track Started")
            (AudioAnalysis|RegisterDefaultSubmix (WorldSubsystems|GetFortAudioAnalysisSubsystem))
            (if (GameplayTags|IsGameplayTagValid (Variables|TrackSettings|GetMusicEventTag))
              (Audio|AddEventbyComponent (GameInstanceSubsystems|GetMusicEventSubsystem) (CallFunction|GetAudioComponent) (Variables|TrackSettings|GetMusicEventTag) "(GameplayTags=)" 0 ""))))))))

(fn QueueTrackStop
  (Variables|Default|SetbQueueStop true)
  (Development|PrintString "Track Queued for Stop")
  (Audio|RemoveEvent (GameInstanceSubsystems|GetMusicEventSubsystem) (CallFunction|GetAudioComponent))
  (AudioAnalysis|UnRegisterDefaultSubmix (WorldSubsystems|GetFortAudioAnalysisSubsystem)))

(fn SetTrackParameter
  (params (ParameterIndex :type Integer) (InFloat :type Float%20%28double-precision%29))
  (Audio|Parameter|SetFloatParameter (CallFunction|GetAudioComponent) (Utilities|Array|Get(aref) (Variables|TrackSettings|GetTrackParameters) ParameterIndex) InFloat :!class /Script/AudioExtensions.AudioParameterControllerInterface))
