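; Quartz music track component that drives adaptive music from a gameplay data tracker component of a configured class.
; It binds through the tracker manager, retrying until available; event values start timed playback requests and a sound mix, with a delayed stop against thrashing.
; Updates interpolate tracker values with attack/release speeds into track parameters.
; It also tracks configured pawns for a one-shot quantized end stinger that fades the music near silence.
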(event Custom|Retry%20Get%20Data%20Tracker%20Manager
  (bind _returnvalue (FortGameplayDataTrackerComponent|GetFortGameplayDataTrackerComponentManager))
  (bind _output_get (Variables|Default|SetDataTrackerManager _returnvalue))
  (Utilities|IsValid _output_get
    (:Is%20Valid
      (bind _outdatatrackercomponents (Class|FortGameplayDataTrackerComponentManager|GetRegisteredDataTrackerComponents (Variables|Default|GetDataTrackerManager)))
      (for _array_element _outdatatrackercomponents
        (if (and (Utilities|Equal(Class) (Utilities|GetClass _array_element) (Variables|DataTrackerClass|GetDataTrackerClass)) (not (Variables|DataTracker|GetbDataTrackerAssigned)))
          (Variables|DataTracker|SetbDataTrackerAssigned true)
          (bind _output_get_2 (Variables|Default|SetDataTrackerComponent _array_element))
          (EventDispatchers|BindEventtoOnChangedHasAnyEventValue _output_get_2 (EventDispatchers|CreateEvent :!event On%20Changed%20Has%20Any%20Event%20Value))))
      (EventDispatchers|BindEventtoOnDataTrackerComponentRegistered (Variables|Default|GetDataTrackerManager) (EventDispatchers|CreateEvent :!event OnDataTrackerComponentRegistered))
      (EventDispatchers|BindEventtoOnDataTrackerComponentUnRegistered (Variables|Default|GetDataTrackerManager) (EventDispatchers|CreateEvent :!event OnDataTrackerComponentUnRegistered)))
    (:Is%20Not%20Valid
      (Utilities|Time|SetTimerbyEvent (EventDispatchers|CreateEvent :!event Custom|Retry%20Get%20Data%20Tracker%20Manager) 1.0))))

(event EventBeginPlay
  (|CallParentFunction)
  (Class|FortAsyncActionCurrentPlaylistReady|PlaylistDataReadyAsync
    (:OnReady
      (bind _returnvalue (FortGameplayDataTrackerComponent|GetFortGameplayDataTrackerComponentManager))
      (bind _output_get (Variables|Default|SetDataTrackerManager _returnvalue))
      (Utilities|IsValid _output_get
        (:Is%20Valid
          (bind _outdatatrackercomponents (Class|FortGameplayDataTrackerComponentManager|GetRegisteredDataTrackerComponents (Variables|Default|GetDataTrackerManager)))
          (for _array_element _outdatatrackercomponents
            (if (and (Utilities|Equal(Class) (Utilities|GetClass _array_element) (Variables|DataTrackerClass|GetDataTrackerClass)) (not (Variables|DataTracker|GetbDataTrackerAssigned)))
              (Variables|DataTracker|SetbDataTrackerAssigned true)
              (bind _output_get_2 (Variables|Default|SetDataTrackerComponent _array_element))
              (EventDispatchers|BindEventtoOnChangedHasAnyEventValue _output_get_2 (EventDispatchers|CreateEvent :!event On%20Changed%20Has%20Any%20Event%20Value))))
          (EventDispatchers|BindEventtoOnDataTrackerComponentRegistered (Variables|Default|GetDataTrackerManager) (EventDispatchers|CreateEvent :!event OnDataTrackerComponentRegistered))
          (EventDispatchers|BindEventtoOnDataTrackerComponentUnRegistered (Variables|Default|GetDataTrackerManager) (EventDispatchers|CreateEvent :!event OnDataTrackerComponentUnRegistered)))
        (:Is%20Not%20Valid
          (Utilities|Time|SetTimerbyEvent (EventDispatchers|CreateEvent :!event Custom|Retry%20Get%20Data%20Tracker%20Manager) 1.0))))))

(fn DataTrackerAddPawns_Internal
  (bind _audiocomponent (CallFunction|GetAudioComponent))
  (Utilities|FlowControl|Sequence
    (:then_0
      (for _array_element (Variables|DataTrackerActorClass|GetGetActorClass)
        (if (Utilities|IsValidClass _array_element)
          (bind _returnvalue (Actor|GetActorOfClass _array_element))
          (Utilities|IsValid _returnvalue
            (:Is%20Valid
              (bind _asfort_pawn (Utilities|Casting|CastToFortPawn _returnvalue))
              (Utilities|Array|Add (Variables|DataTrackerActorClass|GetPawnActorList) _asfort_pawn))))))
    (:then_1
      (if (Utilities|Array|IsNotEmpty (Variables|DataTrackerActorClass|GetPawnActorList))
        (bind _returnvalue_2 (Audio|CreateSound2D (Variables|TrackSettings|Stingers|GetEndStingerSound)))
        (Variables|Default|SetEndStingerAudio _returnvalue_2)
        (Utilities|IsValid _audiocomponent
          (:Is%20Valid
            (Audio|Components|Audio|AdjustVolume _audiocomponent 0.0 1.0 "Linear" :!class /Script/Engine.AudioComponent)))))))

(fn DataTrackerStart_Internal
  (bind _returnvalue (QuartzClockHandle|IsQuartzEnabled (WorldSubsystems|GetQuartzSubsystem)))
  (if _returnvalue
    (Utilities|FlowControl|Sequence
      (:then_0
        (bind _returnvalue_2 (Utilities|Time|SetTimerbyEvent (EventDispatchers|CreateEvent :!event DataTrackerUpdate_Internal) (Variables|Default|GetUpdateFrequency) true))
        (Variables|Default|SetTimerHandle _returnvalue_2)
        (if (not (Variables|DataTracker|GetbSoundMixActive))
          (Audio|PushSoundMixModifier (Variables|DataTracker|GetDataTrackerSoundMix))
          (Variables|DataTracker|SetbSoundMixActive true)))
      (:then_1
        (bind _result (Class|BPQuartzMusicController|RequestTrackPlayback (Variables|Default|GetControllerActor) self))
        (if _result
          (Class|BPQuartzMusicController|PlayTrack (Variables|Default|GetControllerActor) self)))
      (:then_2
        (Override|OnDataTrackerStart)))))

(fn DataTrackerStop_Internal
  (Utilities|Time|ClearandInvalidateTimerbyHandle (Variables|Default|GetTimerHandle))
  (Class|BPQuartzMusicController|StopCurrentTrack (Variables|Default|GetControllerActor)))

(fn DataTrackerUpdate_Internal
  (Utilities|FlowControl|Sequence
    (:then_0
      (Variables|Default|GetDataTrackerManager
        (:then)
        (:else
          (CallFunction|RetryGetDataTrackerManager))))
    (:then_1
      (bind _result (Class|BPQuartzMusicController|RequestTrackPlayback (Variables|Default|GetControllerActor) self))
      (if _result
        (Class|BPQuartzMusicController|PlayTrack (Variables|Default|GetControllerActor) self)))
    (:then_2
      (Override|OnDataTrackerUpdate))
    (:then_3
      (for _index (range (+ (Utilities|Array|LastIndex (Variables|DataTrackerParameters|GetDataTrackerParameters)) 1))
        (Utilities|Struct|SetmembersinDataTrackerParameterData (Utilities|Array|Get(aref) (Variables|DataTrackerParameters|GetDataTrackerParameters) _index) (Math|Interpolation|FInterpTo (.Value_5_E5013D364BA60BD90AECCF88EFF313A4 (Utilities|Struct|BreakDataTrackerParameterData (Utilities|Array|Get(aref) (Variables|DataTrackerParameters|GetDataTrackerParameters) _index))) (Utilities|Select :Index (Utilities|Struct|BreakDataTrackerParameterData (Utilities|Array|Get(aref) (Variables|DataTrackerParameters|GetDataTrackerParameters) _index)) :NewEnumerator0 (Class|FortGameplayDataTrackerComponent|GetEventValue (CallFunction|GetDataTrackerComponent) (.Tag_8_61F575734DB64A573924BCAFD5B4423F (Utilities|Struct|BreakDataTrackerParameterData (Utilities|Array|Get(aref) (Variables|DataTrackerParameters|GetDataTrackerParameters) _index)))) :NewEnumerator1 (Class|FortGameplayDataTrackerComponent|GetAccumulationValue (CallFunction|GetDataTrackerComponent) (.Tag_8_61F575734DB64A573924BCAFD5B4423F (Utilities|Struct|BreakDataTrackerParameterData (Utilities|Array|Get(aref) (Variables|DataTrackerParameters|GetDataTrackerParameters) _index))))) (Utilities|Time|GetWorldDeltaSeconds) (select (> (.Value_5_E5013D364BA60BD90AECCF88EFF313A4 (Utilities|Struct|BreakDataTrackerParameterData (Utilities|Array|Get(aref) (Variables|DataTrackerParameters|GetDataTrackerParameters) _index))) (Utilities|Select :Index (Utilities|Struct|BreakDataTrackerParameterData (Utilities|Array|Get(aref) (Variables|DataTrackerParameters|GetDataTrackerParameters) _index)) :NewEnumerator0 (Class|FortGameplayDataTrackerComponent|GetEventValue (CallFunction|GetDataTrackerComponent) (.Tag_8_61F575734DB64A573924BCAFD5B4423F (Utilities|Struct|BreakDataTrackerParameterData (Utilities|Array|Get(aref) (Variables|DataTrackerParameters|GetDataTrackerParameters) _index)))) :NewEnumerator1 (Class|FortGameplayDataTrackerComponent|GetAccumulationValue (CallFunction|GetDataTrackerComponent) (.Tag_8_61F575734DB64A573924BCAFD5B4423F (Utilities|Struct|BreakDataTrackerParameterData (Utilities|Array|Get(aref) (Variables|DataTrackerParameters|GetDataTrackerParameters) _index)))))) (.ReleaseSpeed_14_3B43A34E49AB27C329C287B6FDF1FDE7 (Utilities|Struct|BreakDataTrackerParameterData (Utilities|Array|Get(aref) (Variables|DataTrackerParameters|GetDataTrackerParameters) _index))) (.AttackSpeed_13_C792F853487F8768F5EC91A10AE35BF9 (Utilities|Struct|BreakDataTrackerParameterData (Utilities|Array|Get(aref) (Variables|DataTrackerParameters|GetDataTrackerParameters) _index))))))))))

(fn EventOnStinger
  (params (EventType :type EQuartzCommandDelegateSubType%20Enum :default "CommandOnFailedToQueue") (Name :type Name))
  (bind _audiocomponent (CallFunction|GetAudioComponent))
  ;[1 1 1] On stinger quantized event - Check for event has started ththen fade out audio component
  (if (Utilities|Enum|Equal(Enum) EventType "CommandOnStarted")
    ;;[1 1 1] Set volume to 0.01 instead of zero to have music start back up if pawn respawns.
    (Utilities|IsValid _audiocomponent
      (:Is%20Valid
        (Audio|Components|Audio|AdjustVolume _audiocomponent 1.5 0.01 "Linear" :!class /Script/Engine.AudioComponent)))))

(fn GetPawnActorList
  (return (Variables|DataTrackerActorClass|GetPawnActorList)))

(fn GetDataTrackerComponent
  (return (Variables|Default|GetDataTrackerComponent)))

(fn GetDataTrackerParameterValue
  (params (Index :type Integer))
  (bind (_type_4_cf035c87466b64cd227da6806fc20aa0 _tag_8_61f575734db64a573924bcafd5b4423f _value_5_e5013d364ba60bd90aeccf88eff313a4) (Utilities|Struct|BreakDataTrackerParameterData (Utilities|Array|Get(aref) (Variables|DataTrackerParameters|GetDataTrackerParameters) Index)))
  (bind _datatrackerparameters (Variables|DataTrackerParameters|GetDataTrackerParameters))
  (if (Utilities|Array|IsValidIndex _datatrackerparameters Index)
    (return _value_5_e5013d364ba60bd90aeccf88eff313a4)))

(fn GetDataTrackerParameterValueTotal
  (for _array_element (Variables|DataTrackerParameters|GetDataTrackerParameters)
    (LocalVariables|SetTotal (+ (.Value_5_E5013D364BA60BD90AECCF88EFF313A4 (Utilities|Struct|BreakDataTrackerParameterData _array_element)) (LocalVariables|GetTotal))))
  (return (LocalVariables|GetTotal)))

(fn OnChangedHasAnyEventValue
  (params (DataTrackerComponent :type Fort%20Gameplay%20Data%20Tracker%20Component%20Object%20Reference) (bHasEventValues :type Boolean))
  (Utilities|FlowControl|Sequence
    (:then_0
      (if bHasEventValues
        (Internal|DataTrackerStartInternal)
        (Utilities|Time|ClearandInvalidateTimerbyHandle (Variables|Default|GetStopTimer))
        (else
          (bind _returnvalue (Utilities|Time|SetTimerbyEvent (EventDispatchers|CreateEvent :!event DataTrackerStop_Internal) 2.0)) ; This is needed because this event thrashes a bit
          (Variables|Default|SetStopTimer _returnvalue)
          (if (Variables|DataTracker|GetbSoundMixActive)
            (Audio|PopSoundMixModifier (Variables|DataTracker|GetDataTrackerSoundMix))
            (Variables|DataTracker|SetbSoundMixActive false)))))
    (:then_1
      ;[1 1 1] Update pawn to check the status and get last value that is greater than zero
      (CallFunction|OnPawnUpdate)
      (CallFunction|SetLastValue bHasEventValues))))

(fn OnDataTrackerComponentRegistered
  (params (RegisteredDataTracker :type Fort%20Gameplay%20Data%20Tracker%20Component%20Object%20Reference))
  (if (and (Utilities|Equal(Class) (Utilities|GetClass RegisteredDataTracker) (Variables|DataTrackerClass|GetDataTrackerClass)) (not (Variables|DataTracker|GetbDataTrackerAssigned)))
    (Variables|DataTracker|SetbDataTrackerAssigned true)
    (bind _output_get (Variables|Default|SetDataTrackerComponent RegisteredDataTracker))
    (EventDispatchers|BindEventtoOnChangedHasAnyEventValue _output_get (EventDispatchers|CreateEvent :!event On%20Changed%20Has%20Any%20Event%20Value))
    (Development|PrintString (Utilities|String|Append "Quartz DataTracker Reg" (Utilities|GetDisplayName RegisteredDataTracker)))))

(fn OnDataTrackerComponentUnRegistered
  (params (UnRegisteredDataTracker :type Fort%20Gameplay%20Data%20Tracker%20Component%20Object%20Reference))
  (if (Utilities|Equal(Class) (Utilities|GetClass UnRegisteredDataTracker) (Variables|DataTrackerClass|GetDataTrackerClass))
    (Variables|DataTracker|SetbDataTrackerAssigned false)))

(fn OnDataTrackerStart
  (Internal|DataTrackerAddPawnsInternal))

(fn OnDataTrackerStop
  (bind _pawnactorlist (Variables|DataTrackerActorClass|GetPawnActorList))
  (if (Utilities|Array|IsNotEmpty _pawnactorlist)
    (Utilities|Array|Clear _pawnactorlist)))

(fn OnDataTrackerUpdate
  (bind _datatrackerparameters (Variables|DataTrackerParameters|GetDataTrackerParameters))
  (bind _returnvalue (Utilities|Array|LastIndex _datatrackerparameters))
  (if (== _returnvalue (Utilities|Array|LastIndex (Variables|TrackSettings|GetTrackParameters)))
    (for _index (range (+ _returnvalue 1))
      (Control|SetTrackParameter _index (CallFunction|GetDataTrackerParameterValue _index)))
    (elif (Variables|DataTracker|GetbHasFiredOffDataTrackerParameterWarning)
      _
      (else
        (Development|PrintString "Data Tracker Parameters array size does not match Track Parameters" true true "(R=1.000000,G=0.061242,B=0.000000,A=1.000000)" 1e+01)
        (Variables|DataTracker|SetbHasFiredOffDataTrackerParameterWarning true)))))

(fn OnPawnUpdate
  ;[1 1 1] Check to see if pawn is dead, once it dies play end stinger quantized. Stinger should play at proper volume based on value
  (for _array_element (Variables|DataTrackerActorClass|GetPawnActorList)
    (if (FortPawn|IsDead _array_element)
      (Utilities|Array|RemoveItem (Variables|DataTrackerActorClass|GetPawnActorList) _array_element)))
  (if (Utilities|Array|IsEmpty (Variables|DataTrackerActorClass|GetPawnActorList))
    _)
  (Utilities|FlowControl|DoOnce false
    (:Completed
      (bind _endstingeraudio (Variables|Default|GetEndStingerAudio)
        (:then
          (Audio|Components|Audio|SetVolumeMultiplier _endstingeraudio (Variables|DataTracker|GetLastValue) :!class /Script/Engine.AudioComponent)
          (Audio|Components|Audio|PlayQuantized _endstingeraudio (CallFunction|GetClockHandle) (Utilities|Struct|MakeQuartzQuantizationBoundary (Variables|TrackSettings|Stingers|GetEndStingerQuantization)) (EventDispatchers|CreateEvent :!event EventOnStinger)))
        (:else)))))

(fn SetLastValue
  (params (HasValue :type Boolean))
  (if HasValue
    ;[1 1 1] Set volume for stinger to play
    (Variables|DataTracker|SetLastValue (CallFunction|GetDataTrackerParameterValueTotal))))
